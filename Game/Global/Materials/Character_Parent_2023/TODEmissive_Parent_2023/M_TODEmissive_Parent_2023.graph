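# Parent character material that makes emissive color react to time of day, wrapped between the 2023 character parent PreFX and PostFX functions.
# Morning, day, evening and night tints feed Color_Time_of_Day, which scales the emissive, optionally limited by a selectable mask channel.
# The brightest channel of the resulting emissive is inverted to darken base color by an adjustable amount.
# Both results are written back into the material attributes.
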Material /Game/Global/Materials/Character_Parent_2023/TODEmissive_Parent_2023/M_TODEmissive_Parent_2023
graph {
 properties {
  bUseMaterialAttributes=true
  bUsedWithClothing=true
  bUsedWithMeshDeformer=true
  bUsedWithMorphTargets=true
  bUsedWithNanite=true
  bUsedWithNiagaraMeshParticles=true
  bUsedWithSkeletalMesh=true
  pixelDepthOffsetMode=PDOM_Legacy
  shadingModel=MSM_FromMaterialExpression
  usedShadingModels=Default Lit | Subsurface | Clear Coat | Subsurface Profile | Cloth
 }
 nodes [
  n1 = MaterialFunctionCall() {materialFunction=&MF_CharacterParent_2023_PreFX}
  MaterialFunctionCall(
   Material=SetMaterialAttributes!(MaterialAttributes=NamedRerouteUsage(), Base Color=NamedRerouteUsage(), Emissive Color=NamedRerouteUsage()) {attributeSetTypes [#1, #2]}
   M=@n1.M
   Pre Skinned Local Position=@n1.M
   Distance Blend=@n1.M
   Fresnel=@n1.M
  ) {materialFunction=&MF_CharacterParent_2023_PostFX}
  n2 = GetMaterialAttributes(None=@n1.Result) {attributeGetTypes [#2, #1]}
  NamedRerouteDeclaration(None=@n2.MaterialAttributes) {name=InMat, nodeColor=<rgba 0.3359 1 0 1>}
  NamedRerouteDeclaration(None=@n2.Emissive Color) {name=InEmissive, nodeColor=<rgba 1 0 0.6562 1>}
  NamedRerouteDeclaration(None=@n1.Tech Art Mask) {name=TechArtMask, nodeColor=<rgba 0.2188 0 1 1>}
  NamedRerouteDeclaration(None=@n2.Base Color) {name=InBC, nodeColor=<rgba 0.1016 1 0 1>}
  n3 = VectorParameter() {group=10_TODEmissiveReactivity, parameterName=Emissive_MorningTint, sortPriority=0, defaultValue=<rgba 1 1 1 1>}
  n4 = VectorParameter() {group=10_TODEmissiveReactivity, parameterName=Emissive_DayTint, sortPriority=0, defaultValue=<rgba 1 1 1 1>}
  n5 = VectorParameter() {group=10_TODEmissiveReactivity, parameterName=Emissive_EveningTint, sortPriority=0, defaultValue=<rgba 1 1 1 1>}
  n6 = VectorParameter() {group=10_TODEmissiveReactivity, parameterName=Emissive_NightTint, sortPriority=0, defaultValue=<rgba 1 1 1 1>}
  n7 = MaterialFunctionCall(Color_Morning=@n3.RGB, Color_Day=@n4.RGB, Color_Evening=@n5.RGB, Color_Night=@n6.RGB) {materialFunction=&Color_Time_of_Day}
  n8 = StaticSwitchParameter(
   True=ChannelMaskParameter(None=NamedRerouteUsage()) {group=10_TODEmissiveReactivity, parameterName=TODEmissiveReactivityMaskChannel}.R
   False=Constant() {r=1}
  ) {group=10_TODEmissiveReactivity, parameterName=UseCustomMask}
  n9 = NamedRerouteDeclaration(None=Multiply(A=NamedRerouteUsage(), B=Multiply(A=@n7.Result, B=@n8))) {name=OutEmissive, nodeColor=<rgba 1 0 0.4219 1>}
  n10 = Max(A=ComponentMask(None=@n9) {r=true}, B=ComponentMask(None=@n9) {g=true})
  n11 = ComponentMask(None=@n9) {b=true}
  n12 = Saturate(None=OneMinus(None=Max(A=@n10, B=@n11)))
  n13 = NamedRerouteUsage()
  NamedRerouteDeclaration(None=LinearInterpolate(
    A=NamedRerouteUsage()
    B=Multiply(A=@n12, B=@n13)
    Alpha=ScalarParameter() {group=10_TODEmissiveReactivity, parameterName=BaseColor_DynamicDarkeningAmount}
   )) {name=OutBC, nodeColor=<rgba 0 0.7422 1 1>}
 ]
}
